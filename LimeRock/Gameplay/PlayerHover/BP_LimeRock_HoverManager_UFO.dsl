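; Manages UFO hover for players in LimeRock, using two capsule pickup areas that are enabled or disabled by events.
; Entering a capsule adds the pawn to a hovering list and, on authority, applies the hover GameplayEffect; exiting removes its VFX and list entry.
; The looping hover GameplayCue is added and removed manually because effect-applied VFX failed to clear.
; All-player events apply hover or clear VFX, the latter broadcasting a hover router event.
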(event OnComponentBeginOverlap(Capsule-PlayerDetector-UFOStop1)
  (params (OverlappedComponent :type Primitive%20Component%20Object%20Reference) (OtherActor :type Actor%20Object%20Reference) (OtherComp :type Primitive%20Component%20Object%20Reference) (OtherBodyIndex :type Integer) (bFromSweep :type Boolean) (SweepResult :type Hit%20Result%20Structure%20%28by%20ref%29))
  ; Trigger Volumes
  (CallFunction|PlayerTryUFOHover OtherActor))

(event OnComponentEndOverlap(Capsule-PlayerDetector-UFOStop1)
  (params (OverlappedComponent :type Primitive%20Component%20Object%20Reference) (OtherActor :type Actor%20Object%20Reference) (OtherComp :type Primitive%20Component%20Object%20Reference) (OtherBodyIndex :type Integer))
  ; Trigger Volumes
  (bind _asfort_player_pawn (Utilities|Casting|CastToFortPlayerPawn OtherActor))
  (CallFunction|RemoveUFOHoverVFX-SinglePlayer _asfort_player_pawn)
  (Utilities|Array|RemoveItem (Variables|Default|GetHovering%20Players%20List) _asfort_player_pawn))

(event Custom|UFO%20Pickup%20Area%201%20-%20Set%20Enabled
  (params (Enabled :type Boolean))
  ; Trigger Volumes
  (Collision|SetCollisionEnabled (Variables|Default|GetCapsule%20-%20Player%20Detector%20-%20UFO%20Stop%201) (select Enabled "QueryOnly" "NoCollision"))
  (if Enabled
    (Variables|Default|SetHoverZoneActiveIndex 1)
    (else
      (CallFunction|LRRemoveHoverVFXfromPlayers)
      (Utilities|Array|Clear (Variables|Default|GetHovering%20Players%20List))
      (Variables|Default|SetHoverZoneActiveIndex 0))))

(event OnComponentEndOverlap(Capsule-PlayerDetector-UFOStop2)
  (params (OverlappedComponent :type Primitive%20Component%20Object%20Reference) (OtherActor :type Actor%20Object%20Reference) (OtherComp :type Primitive%20Component%20Object%20Reference) (OtherBodyIndex :type Integer))
  ; Trigger Volumes
  (bind _asfort_player_pawn (Utilities|Casting|CastToFortPlayerPawn OtherActor))
  (CallFunction|RemoveUFOHoverVFX-SinglePlayer _asfort_player_pawn)
  (Utilities|Array|RemoveItem (Variables|Default|GetHovering%20Players%20List) _asfort_player_pawn))

(event OnComponentBeginOverlap(Capsule-PlayerDetector-UFOStop2)
  (params (OverlappedComponent :type Primitive%20Component%20Object%20Reference) (OtherActor :type Actor%20Object%20Reference) (OtherComp :type Primitive%20Component%20Object%20Reference) (OtherBodyIndex :type Integer) (bFromSweep :type Boolean) (SweepResult :type Hit%20Result%20Structure%20%28by%20ref%29))
  ; Trigger Volumes
  (CallFunction|PlayerTryUFOHover OtherActor))

(event Custom|UFO%20Pickup%20Area%202%20-%20Set%20Enabled
  (params (Enabled :type Boolean))
  ; Trigger Volumes
  (Collision|SetCollisionEnabled (Variables|Default|GetCapsule%20-%20Player%20Detector%20-%20UFO%20Stop%202) (select Enabled "QueryOnly" "NoCollision"))
  (if Enabled
    (Variables|Default|SetHoverZoneActiveIndex 2)
    (else
      (CallFunction|LRRemoveHoverVFXfromPlayers)
      (Utilities|Array|Clear (Variables|Default|GetHovering%20Players%20List))
      (Variables|Default|SetHoverZoneActiveIndex 0))))

(event Custom|Player%20Try%20UFO%20Hover
  (params (Object :type Object%20Reference))
  ; Trigger Volumes
  (bind _asfort_player_pawn (Utilities|Casting|CastToFortPlayerPawn Object))
  (if (not (Utilities|Array|ContainsItem (Variables|Default|GetHovering%20Players%20List) _asfort_player_pawn))
    (Utilities|Array|Add (Variables|Default|GetHovering%20Players%20List) _asfort_player_pawn) ; add player to hovering\nplayer list to make sure\nwe don't try to make them\nhover again
    (CallFunction|StartUFOHover-SinglePlayer _asfort_player_pawn)))

(event Custom|LR_ApplyHoverToAllPlayers
  ; Hover - All Players
  (Networking|SwitchHasAuthority
    (:Authority
      (bind _outfortplayerpawns (Fortnite|GetFortPlayerPawns))
      (for _array_element _outfortplayerpawns
        (bind _returnvalue (GameplayEffects|ApplyGameplayEffectToTarget :self (Ability|GetFortAbilitySystemComponent _array_element) :GameplayEffectClass "/LimeRock/Gameplay/PlayerHover/GE_LimeRock_PlayerHover_UFO.GE_LimeRock_PlayerHover_UFO_C" :Target (Ability|GetFortAbilitySystemComponent _array_element) :Level 0.0))
        (Variables|Default|SetGameplayEffectAllPlayers _returnvalue)
        (GameplayCue|AddGameplayCueOnActor(Looping) _array_element (Variables|Default|GetTag%20-%20Hover%20GCNL) (Ability|GameplayCue|MakeGameplayCueParameters 0.0)) ; Adding / Removing GCNL through blueprint as for some reason\nthe VFX weren't removing from the player when the hover GE ended\nwhen applying the GCNL tag through the GE.
      ))))

(event Custom|Start%20UFO%20Hover%20-%20Single%20Player
  (params (Fort%20Player%20Pawn :type Fort%20Player%20Pawn%20Object%20Reference))
  ; UFO Hover - Single Player
  (Networking|SwitchHasAuthority
    (:Authority
      (bind _returnvalue (GameplayEffects|ApplyGameplayEffectToTarget :self (Ability|GetFortAbilitySystemComponent Fort%20Player%20Pawn) :GameplayEffectClass "/LimeRock/Gameplay/PlayerHover/GE_LimeRock_PlayerHover_UFO.GE_LimeRock_PlayerHover_UFO_C" :Target (Ability|GetFortAbilitySystemComponent Fort%20Player%20Pawn) :Level 0.0))
      (Variables|Default|SetGameplayEffectSinglePlayer _returnvalue)
      (GameplayCue|AddGameplayCueOnActor(Looping) Fort%20Player%20Pawn (Variables|Default|GetTag%20-%20Hover%20GCNL) (Ability|GameplayCue|MakeGameplayCueParameters 0.0)) ; Adding / Removing GCNL through blueprint as for some reason\nthe VFX weren't removing from the player when the hover GE ended\nwhen applying the GCNL tag through the GE.
    )))

(event Custom|LR%20Remove%20Hover%20VFX%20From%20Players
  ; Hover - All Players
  (bind _outfortplayerpawns (Fortnite|GetFortPlayerPawns))
  (for _array_element _outfortplayerpawns
    (GameplayCue|RemoveGameplayCueOnActor(Looping) _array_element (Variables|Default|GetTag%20-%20Hover%20GCNL) (Ability|GameplayCue|MakeGameplayCueParameters 0.0))
    (GameplayEventRouter|BroadcastStatefulEvent (GameplayEventRouter|GetGlobalEventRouter) (Utilities|Struct|MakeLimeRockHoverEventRouter "NewEnumerator0") 0)))

(event Custom|Remove%20UFO%20Hover%20VFX%20-%20Single%20Player
  (params (Fort%20Player%20Pawn :type Fort%20Player%20Pawn%20Object%20Reference))
  ; UFO Hover - Single Player
  (GameplayCue|RemoveGameplayCueOnActor(Looping) Fort%20Player%20Pawn (Variables|Default|GetTag%20-%20Hover%20GCNL) (Ability|GameplayCue|MakeGameplayCueParameters 0.0)))

(event Custom|LR_TestHover_GCNL
  (bind _returnvalue (Game|GetPlayerPawn 0))
  (bind _returnvalue_2 (Ability|GameplayCue|MakeGameplayCueParameters 0.0))
  (bind _tag___hover_gcnl (Variables|Default|GetTag%20-%20Hover%20GCNL))
  ; Hover - All Players
  ;; Test Hover GCNL
  (Utilities|FlowControl|FlipFlop
    (:A
      (GameplayCue|AddGameplayCueOnActor(Looping) _returnvalue _tag___hover_gcnl _returnvalue_2))
    (:B
      (GameplayCue|RemoveGameplayCueOnActor(Looping) _returnvalue _tag___hover_gcnl _returnvalue_2))))

(fn ConstructionScript)
